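# Unlit, two-sided translucent trail material for the Caterpillar pickaxe, usable on beam trails and Niagara ribbons.
# Emissive blends greens and yellow-greens using a thin edge mask plus panning noise and smoke, partly rounded into hard bands.
# Opacity combines that edge with sphere-mask shapes, eroded by panning, distorted smoke under Dynamic Parameter control.
# Particle alpha and a lengthwise UV fade scale opacity; depth fade applies outside the mobile shading path.
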Material /BRCosmetics/Weapons/FORT_Melee/Pickaxe_Caterpillar/FX/M_Pickaxe_Trail_Caterpillar
graph {
 properties {
  bUsedWithBeamTrails=true
  bUsedWithNiagaraRibbons=true
  blendMode=BLEND_Translucent
  pixelDepthOffsetMode=PDOM_Legacy
  shadingModel=MSM_Unlit
  twoSided=true
  usedShadingModels=Unlit
 }
 nodes [
  n1 = Constant3Vector() {constant=<rgba 0.4369 0.7188 0 0>}
  n2 = Constant3Vector() {constant=<rgba 1.1277 1.4 0.1531 0>}
  n3 = MaterialFunctionCall(Float2=TextureCoordinate()) {materialFunction=&BreakOutFloat2Components}
  n4 = Constant() {r=0.01}
  n5 = AppendVector(A=@n4, B=@n3.G)
  n6 = ComponentMask(None=TextureCoordinate()) {r=true}
  n7 = LinearInterpolate(
   A=ScalarParameter() {defaultValue=0.02, parameterName=Edge Radius A}
   B=ScalarParameter() {defaultValue=0.005, parameterName=Edge Radius B}
   Alpha=@n6
  )
  n8 = LinearInterpolate(A=ScalarParameter() {defaultValue=0.3, parameterName=Edge Hardness A}, B=ScalarParameter() {parameterName=Edge Hardness B}, Alpha=@n6)
  n9 = Multiply(A=Multiply(A=Multiply(A=@n3.R), B=SphereMask(A=@n5, B=@n4, Radius=@n7, Hardness=@n8))) {constB=1.8}
  n10 = TextureCoordinate() {coordinateIndex=1, uTiling=2}
  n11 = AppendVector(
   A=ComponentMask(None=@n10) {r=true}
   B=LinearInterpolate(A=Constant() {r=0.1}, B=Constant() {r=0.9}, Alpha=ComponentMask(None=@n10) {g=true})
  )
  n12 = TextureCoordinate()
  n13 = VectorParameter() {parameterName=UV, defaultValue=<rgba 0.4 0.7 0 0>}
  n14 = AppendVector(A=@n13.R, B=@n13.R)
  n15 = TextureCoordinate() {uTiling=0.3, vTiling=0.6}
  n16 = ComponentMask(None=TextureSample(UVs=Panner(Coordinate=@n15) {bFractionalPart=true, speedX=-0.2}) {samplerType=SAMPLERTYPE_LinearColor, texture=&T_NoisePack_LinearColor}.RGB) {g=true, r=true}
  n17 = Panner(Coordinate=Add(A=Multiply(A=@n12, B=@n14), B=Multiply(A=@n16) {constB=0.1})) {bFractionalPart=true, speedX=0.45}
  n18 = TextureSample(UVs=@n17) {texture=&T_NoisePack_01}
  n19 = Add(
   A=TextureSample(UVs=Panner(Coordinate=@n11) {speedX=-0.5}) {texture=&T_Inky_Smoke_Tile}.R
   B=Multiply(A=Power(Base=@n18.G)) {constB=0.6}
  )
  n20 = Multiply(A=Round(None=Power(Base=@n19) {constExponent=1.2})) {constB=0.2}
  n21 = TextureSample(UVs=@n17) {samplerType=SAMPLERTYPE_LinearColor, texture=/Game/Effects/Fort_Effects/Textures/Tile/T_Raygun_Muzzle_02}
  n22 = Add(A=@n9, B=Clamp(None=Add(A=@n20, B=@n21.R)))
  n23 = MaterialFunctionCall(Float2=TextureCoordinate()) {materialFunction=&BreakOutFloat2Components}
  MaterialFunctionCall(Emissive=LinearInterpolate(A=LinearInterpolate(A=@n1, B=@n2, Alpha=@n22), B=Constant3Vector() {constant=<rgba 0.0572 1 0 1>}, Alpha=@n23.R)) {materialFunction=&MF_FX_EmissiveAdjustments}
  Multiply(A=@n23.R)
  Power(Base=@n23.R)
  n24 = MaterialFunctionCall(Float2=TextureCoordinate()) {materialFunction=&BreakOutFloat2Components}
  n25 = SphereMask(
   A=@n24.G
   B=Constant() {r=1.04}
   Radius=Add(A=DynamicParameter() {defaultValue=<rgba 0 0 0 1>}.Param3, B=ScalarParameter() {defaultValue=0.12, parameterName=Param_3})
   Hardness=Constant() {r=0.9}
  )
  Round(None=@n25)
  n26 = SphereMask(A=TextureCoordinate() {uTiling=0.5}, B=Constant2Vector() {g=0.82, r=0.85}, Radius=Constant() {r=0.88}, Hardness=Constant() {r=0.7})
  n27 = MaterialFunctionCall(Float2=TextureCoordinate()) {materialFunction=&BreakOutFloat2Components}
  n28 = Multiply(A=Power(Base=OneMinus(None=@n26)) {constExponent=5}, B=OneMinus(None=@n27.G))
  n29 = Clamp(None=Multiply(A=Multiply(A=@n28, B=@n27.R)) {constB=2})
  n30 = TextureCoordinate() {uTiling=0.6}
  n31 = ComponentMask(None=@n30) {r=true}
  n32 = LinearInterpolate(A=Constant() {r=0.1}, B=Constant() {r=0.9}, Alpha=ComponentMask(None=@n30) {g=true})
  n33 = TextureSample(UVs=Panner(Coordinate=AppendVector(A=@n31, B=@n32)) {speedX=-0.3}) {samplerType=SAMPLERTYPE_LinearColor, texture=/Game/Effects/Fort_Effects/Gameplay/SmokeDustExplosions/Textures/T_RibbonSmokeMasked_03}
  n34 = Multiply(A=Multiply(A=Power(Base=@n27.R), B=Power(Base=@n33.R) {constExponent=5})) {constB=5.5}
  n35 = TextureCoordinate() {coordinateIndex=1}
  n36 = ScalarParameter() {defaultValue=0.01, parameterName=Param_1}
  n37 = ScalarParameter() {defaultValue=1.9}
  n38 = ComponentMask(None=@n35) {g=true}
  n39 = AppendVector(
   A=Multiply(A=ScalarParameter() {defaultValue=1.85, parameterName=Param_2}, B=ComponentMask(None=@n35) {r=true})
   B=Clamp(None=LinearInterpolate(A=@n36, B=@n37, Alpha=@n38))
  )
  n40 = Panner(Coordinate=TextureCoordinate() {uTiling=0.3, vTiling=0.6}) {bFractionalPart=true, speedX=-0.2}
  n41 = Multiply(
   A=ComponentMask(None=TextureSample(UVs=@n40) {samplerType=SAMPLERTYPE_Normal, texture=&T_Weap_Distort_N_02}.RGB) {g=true, r=true}
   B=ParticleColor().R
  ) {constB=0.1}
  n42 = TextureSample(UVs=Panner(Coordinate=Add(A=@n39, B=@n41)) {speedX=-0.35}) {samplerType=SAMPLERTYPE_LinearColor, texture=&T_RibbonSmokeMasked_03}
  n43 = Clamp(None=Subtract(A=@n42.R, B=DynamicParameter() {defaultValue=<rgba 0 0 0 1>}.Param2) {constB=0.1})
  n44 = Constant() {r=1.04}
  n45 = DynamicParameter() {defaultValue=<rgba 0 0 0 1>}
  n46 = Constant() {r=0.85}
  n47 = Multiply(
   A=LinearInterpolate(A=DynamicParameter() {defaultValue=<rgba 0 0 0 1>}.Param4, Alpha=Round(None=@n43))
   B=Multiply(A=SphereMask(A=@n24.R, B=@n44, Radius=@n45.Param3, Hardness=@n46), B=@n25)
  )
  n48 = DynamicParameter() {defaultValue=<rgba 0 0 0 1>}
  n49 = Clamp(None=Subtract(A=Subtract(A=@n29, B=@n34), B=Multiply(A=@n47, B=@n48.Param1)))
  n50 = Multiply(A=ParticleColor().A, B=Multiply(A=Add(A=@n9, B=@n49), B=OneMinus(None=@n27.R)))
  ShadingPathSwitch(Default=Multiply(A=@n50, B=DepthFade() {fadeDistanceDefault=50}), Mobile=@n50) {desc=Migrated from FeatureLevelSwitch}
 ]
}
